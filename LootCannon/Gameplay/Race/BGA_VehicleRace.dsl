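; BGA_VehicleRace is a vehicle race actor with helper functions for start conditions and winner lookup.
; Overlaps with the starting box gather candidate vehicles, then checks require a minimum count of vehicles and drivers.
; The movement check fails if any forward speed exceeds a threshold; despite its name, it never tests facing.
; Other functions register driven vehicles as racers and store the winning vehicle's driver; no calling logic appears here.
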(fn CheckPlayersInDrivingSeatOfAtLeastTwoVehicles
  (for _array_element (Variables|Default|GetVehicleActorsForRaceStart)
    (bind _asfort_vehicle_interface (Utilities|Casting|CastToFortVehicleInterface _array_element))
    (Utilities|IsValid (Vehicle|GetDriver _asfort_vehicle_interface)
      (:Is%20Valid
        (Math|Integer|IncrementInt (LocalVariables|GetNumOfDriversLocal)))))
  (return (>= (LocalVariables|GetNumOfDriversLocal) (Variables|Default|GetMinNumVehiclesToStartRace))))

(fn CheckVehiclesAreNotMovingAndFacingCorrectDirection
  (bind _array_element (Utilities|Array|ForEachLoopwithBreak (Variables|Default|GetVehicleActorsForRaceStart) :!label _label)
    (:LoopBody
      (Utilities|IsValid _array_element
        (:Is%20Valid
          (bind _asfort_vehicle_interface (Utilities|Casting|CastToFortVehicleInterface _array_element))
          (Vehicle|GetVehicleActor _asfort_vehicle_interface :!class /Script/FortniteGame.FortVehicleInterface)
          (if (> (Vehicle|GetForwardSpeed _asfort_vehicle_interface) 1e+01)
            (LocalVariables|SetVehicleIsNotInPositionLocal true)
            (goto _label Break)))))
    (:Completed
      (if (LocalVariables|GetVehicleIsNotInPositionLocal)
        (return false)
        (else
          (return true))))))

(fn GetAllVehiclesInTheRace
  (for _array_element (Collision|GetOverlappingActors (Variables|Default|GetRaceStartingBox) "/Script/Engine.Actor" :!class /Script/Engine.PrimitiveComponent)
    (bind _asfort_vehicle_interface (Utilities|Casting|CastToFortVehicleInterface _array_element))
    (Utilities|IsValid (Vehicle|GetDriver _asfort_vehicle_interface)
      (:Is%20Valid
        (Utilities|Array|AddUnique (Variables|Default|GetVehicleActorsInRace) _array_element)))))

(fn GetOverlappingVehicles
  (for _array_element (Collision|GetOverlappingActors (Variables|Default|GetRaceStartingBox) "/Script/Engine.Actor" :!class /Script/Engine.PrimitiveComponent)
    (Utilities|Casting|CastToFortVehicleInterface _array_element)
    (Utilities|Array|AddUnique (Variables|Default|GetVehicleActorsForRaceStart) _array_element))
  (return (>= (Utilities|Array|Length (Variables|Default|GetVehicleActorsForRaceStart)) (Variables|Default|GetMinNumVehiclesToStartRace))))

(fn GetPlayerFromWinningVehicle
  (bind _winningvehicleref (Variables|Default|GetWinningVehicleRef))
  (Utilities|IsValid _winningvehicleref
    (:Is%20Valid
      (bind _asfort_vehicle_interface (Utilities|Casting|CastToFortVehicleInterface _winningvehicleref))
      (Variables|Default|SetWinningPlayerPawnRef (Vehicle|GetDriver _asfort_vehicle_interface)))))

(fn ConstructionScript)
