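; Vehicle race start light Blueprint that shows the countdown state by changing a material on its light.
; The LightState RepNotify switches on the integer and sets slot 0 of Spotlight1 to the off, red or green material.
; States 1 to 3 all apply the same red to that single spotlight, and state 4 turns it green.
; Nothing here sets LightState, and the construction script is empty.
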(fn OnRep_LightState
  (switch int (Variables|Default|GetLightState)
    (:0
      (Rendering|Material|SetMaterial (Variables|Default|GetSpotlight1) 0 (Variables|Default|GetMaterialOff)) ; all off
    )
    (:1
      (Rendering|Material|SetMaterial (Variables|Default|GetSpotlight1) 0 (Variables|Default|GetMaterialRed)))
    (:2
      (Rendering|Material|SetMaterial (Variables|Default|GetSpotlight1) 0 (Variables|Default|GetMaterialRed)))
    (:3
      (Rendering|Material|SetMaterial (Variables|Default|GetSpotlight1) 0 (Variables|Default|GetMaterialRed)))
    (:4
      (Rendering|Material|SetMaterial (Variables|Default|GetSpotlight1) 0 (Variables|Default|GetMaterialGreen)))))

(fn ConstructionScript)
